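# Material function that picks the friend-or-foe tech color for team-colored materials in Fortnite.
# It lerps the FriendlyTech and EnemyTech colors from the shared material parameter collection by the FriendOrFoe value, outputting RGB plus that value.
# FriendOrFoe defaults to a same-named scalar parameter.
# A static debug switch makes that default flip between 0 and 1 over time at an adjustable speed.
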MaterialFunction /Game/Packages/Fortress_SharedMaterials/Base_Material_Functions/Fortnite_FriendOrFoe_TechColor
graph {
 properties {bExposeToLibrary=true, libraryCategoriesText [Misc,Fortnite]}
 nodes [
  n1 = FunctionInput() {bUsePreviewValueAsDefault=true, inputName=Debug speed, inputType=FunctionInput_Scalar, sortPriority=2, previewValue=<0.25 0 0 1>}
  n2 = Subtract(A=MaterialFunctionCall(Speed=@n1) {materialFunction=&TimeWithSpeedVariable}.Frac Time) {constB=0.5}
  n3 = FunctionInput(Preview=StaticBool()) {
   bUsePreviewValueAsDefault=true
   description=Switches back and forth between the two team colors if nothing is entered in the team color input
   inputName=Debug Colors
   inputType=FunctionInput_StaticBool
   sortPriority=1
  }
  n4 = FunctionInput(Preview=StaticSwitch(True=Ceil(None=@n2), False=ScalarParameter() {group=Team Colors, parameterName=FriendOrFoe}, Value=@n3)) {
   bUsePreviewValueAsDefault=true
   description=0 is red and 1 is blue. The default value is a parameter called team with a value of 0.
   inputName=FriendOrFoe
   inputType=FunctionInput_Scalar
  }
  FunctionOutput(None=@n4) {outputName=Friend or Foe, sortPriority=1}
  n5 = CollectionParameter() {collection=&FortniteMaterialParameters, parameterName=FriendlyTech}
  n6 = CollectionParameter() {collection=&FortniteMaterialParameters, parameterName=EnemyTech}
  FunctionOutput(None=ComponentMask(None=LinearInterpolate(A=@n5, B=@n6, Alpha=@n4)) {b=true, g=true, r=true})
 ]
}
