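# Translucent unlit material that fakes a projected decal of the air strike gadget's grid, with desaturated friend-or-foe coloring.
# It rebuilds object-space coordinates from the world position behind translucency, then samples the grid texture tiled and untiled to combine lines with a gradient.
# Opacity multiplies that pattern by per-axis box falloffs, a camera-proximity fade and an upward-facing mask from scene world normals.
# Shading path switches substitute constants for the normal mask on forward and mobile.
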Material /Game/Gadgets/Assets/AirStrike/Materials/FakeGridDecal
graph {
 properties {blendMode=BLEND_Translucent, pixelDepthOffsetMode=PDOM_Legacy, shadingModel=MSM_Unlit, usedShadingModels=Unlit}
 nodes [
  Desaturation(
   None=MaterialFunctionCall!(FriendOrFoe, Debug Colors, Debug speed) {materialFunction=/Game/Packages/Fortress_SharedMaterials/Base_Material_Functions/Fortnite_FriendOrFoe_TechColor}.Result
   Fraction=ScalarParameter() {defaultValue=0.0907, parameterName=Desaturation}
  )
  n1 = MaterialFunctionCall() {materialFunction=/Engine/Functions/Engine_MaterialFunctions02/WorldPositionOffset/ComponentBasisVectors}
  n2 = Subtract(
   A=MaterialFunctionCall() {materialFunction=&WorldPositionBehindTranslucency}.Result
   B=MaterialFunctionCall() {materialFunction=&ObjectPivotPoint}.Object Pivot Location
  )
  n3 = DotProduct(A=Normalize(VectorInput=@n1.X), B=@n2)
  n4 = DotProduct(A=Normalize(VectorInput=@n1.Y), B=@n2)
  n5 = MaterialFunctionCall() {materialFunction=/Engine/Functions/Engine_MaterialFunctions02/WorldPositionOffset/ObjectScale}
  n6 = ComponentMask(None=Divide(A=AppendVector(A=@n3, B=@n4), B=AppendVector(A=@n5.Scale X, B=@n5.Scale X))) {g=true, r=true}
  n7 = Multiply(A=@n6, B=ScalarParameter() {defaultValue=3, parameterName=GridRepeat})
  n8 = TextureSample(UVs=@n6) {samplerType=SAMPLERTYPE_LinearColor, texture=/Game/Gadgets/Assets/AirStrike/Textures/T_Airstrike_Grid_01}
  n9 = Multiply(
   A=Subtract(
    A=TextureSample(UVs=@n7) {samplerType=SAMPLERTYPE_LinearColor, texture=&T_Airstrike_Grid_01}.R
    B=@n8.G
   )
   B=ScalarParameter() {defaultValue=0.6062, parameterName=Grid Brightness}
  )
  n10 = Multiply(A=OneMinus(None=@n8.G), B=ScalarParameter() {defaultValue=0.6018, parameterName=Gradient Brightness})
  n11 = Multiply(A=Clamp(None=Add(A=@n9, B=@n10)), B=ScalarParameter() {defaultValue=0.0731, parameterName=Opacity})
  n12 = DotProduct(
   A=ComponentMask(None=SceneTexture() {sceneTextureId=PPI_WorldNormal}.Color) {b=true, g=true, r=true}
   B=Constant3Vector() {constant=<rgba 0 0 1 0>}
  )
  n13 = ShadingPathSwitch(Default=@n12, Deferred=@n12, Forward=Constant3Vector() {constant=<rgba 0 0 1 0>}, Mobile=@n12)
  n14 = ShadingPathSwitch(Default=Power(Base=Clamp(None=@n13)), Mobile=Constant() {r=1}) {desc=Migrated from FeatureLevelSwitch}
  n15 = MaterialFunctionCall(Float2=@n6) {materialFunction=&BreakOutFloat2Components}
  n16 = Constant() {r=0.5}
  n17 = ScalarParameter() {defaultValue=0.9, parameterName=X Opacity Hardness}
  n18 = ScalarParameter() {defaultValue=0.9, parameterName=Y Opacity Hardness}
  n19 = DotProduct(A=Normalize(VectorInput=@n1.Z), B=@n2)
  n20 = Multiply(
   A=Multiply(A=SphereMask(A=@n15.R, B=@n16, Hardness=@n17) {attenuationRadius=0.5}, B=SphereMask(A=@n15.G, B=@n16, Hardness=@n18) {attenuationRadius=0.5})
   B=SphereMask(A=Divide(A=@n19, B=@n5.Scale Z), B=Constant() {r=0.5}, Hardness=ScalarParameter() {defaultValue=0.5, parameterName=Z Falloff Hardness}) {attenuationRadius=0.5}
  )
  n21 = OneMinus(None=SphereMask(A=CameraPositionWS(), B=WorldPosition().XYZ) {hardnessPercent=0})
  Multiply(A=Multiply(A=Multiply(A=@n11, B=@n14), B=Multiply(A=@n20, B=@n21)))
  SceneTexture() {sceneTextureId=PPI_DecalMask}
 ]
}
