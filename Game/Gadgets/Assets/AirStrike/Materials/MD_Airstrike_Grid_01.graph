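# Translucent unlit deferred decal material for the Airstrike gadget's grid marking, with a friend-or-foe tech color that is partly desaturated.
# One grid texture is sampled twice: tiled red by GridRepeat, minus untiled green, forms the grid, while inverted untiled green forms a gradient.
# Grid and gradient get separate brightness parameters, then are summed, clamped and scaled by an Opacity parameter for a faint default intensity.
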Material /Game/Gadgets/Assets/AirStrike/Materials/MD_Airstrike_Grid_01
graph {
 properties {
  blendMode=BLEND_Translucent
  materialDomain=MD_DeferredDecal
  pixelDepthOffsetMode=PDOM_Legacy
  shadingModel=MSM_Unlit
  usedShadingModels=Unlit
 }
 nodes [
  Constant()
  Desaturation(
   None=MaterialFunctionCall!(FriendOrFoe, Debug Colors, Debug speed) {materialFunction=/Game/Packages/Fortress_SharedMaterials/Base_Material_Functions/Fortnite_FriendOrFoe_TechColor}.Result
   Fraction=ScalarParameter() {defaultValue=0.0907, parameterName=Desaturation}
  )
  n1 = Multiply(A=TextureCoordinate(), B=ScalarParameter() {defaultValue=3, parameterName=GridRepeat})
  n2 = TextureSample() {samplerType=SAMPLERTYPE_LinearColor, texture=/Game/Gadgets/Assets/AirStrike/Textures/T_Airstrike_Grid_01}
  n3 = Multiply(
   A=Subtract(
    A=TextureSample(UVs=@n1) {samplerType=SAMPLERTYPE_LinearColor, texture=&T_Airstrike_Grid_01}.R
    B=@n2.G
   )
   B=ScalarParameter() {defaultValue=0.6062, parameterName=Grid Brightness}
  )
  n4 = Multiply(A=OneMinus(None=@n2.G), B=ScalarParameter() {defaultValue=0.6018, parameterName=Gradient Brightness})
  Multiply(A=Clamp(None=Add(A=@n3, B=@n4)), B=ScalarParameter() {defaultValue=0.0731, parameterName=Opacity})
 ]
}
