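# Unlit translucent material for the air strike gadget's reticule, usable on mesh particles with responsive AA.
# FriendOrFoe blends Color and Color B for team coloring, then particle color tints the result before MF_FX_EmissiveAdjustments.
# Panned screen-space Y becomes fine scrolling scanlines, softened by multiplying the fraction by its inverse.
# Those lines are multiplied by MF_Fresnel_FX and particle alpha to form a transparency mask.
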Material /Game/Gadgets/Assets/AirStrike/Materials/M_Airstike_ReticuleA_01
graph {
 properties {
  bEnableResponsiveAA=true
  bUsedWithMeshParticles=true
  blendMode=BLEND_Translucent
  pixelDepthOffsetMode=PDOM_Legacy
  shadingModel=MSM_Unlit
  usedShadingModels=Unlit
 }
 nodes [
  n1 = VectorParameter() {parameterName=Color, defaultValue=<rgba 0.54 3.1315 6 0>}
  n2 = VectorParameter() {parameterName=Color B, defaultValue=<rgba 20 11.6996 2.7071 0>}
  n3 = ScalarParameter() {group=Team Colors, parameterName=FriendOrFoe}
  n4 = ParticleColor()
  MaterialFunctionCall(Emissive=Multiply(A=LinearInterpolate(A=@n1.RGB, B=@n2.RGB, Alpha=@n3), B=@n4.RGB)) {materialFunction=&MF_FX_EmissiveAdjustments}
  n5 = Panner(Coordinate=ScreenPosition().ViewportUV) {bFractionalPart=true, speedY=3}
  n6 = Constant2Vector() {g=150, r=1}
  n7 = Frac(None=ComponentMask(None=Multiply(A=@n5, B=@n6)) {g=true})
  n8 = OneMinus(None=@n7)
  Multiply(
   A=@n4.A
   B=Multiply(A=MaterialFunctionCall() {materialFunction=&MF_Fresnel_FX}.Fresnel, B=Multiply(A=@n7, B=@n8))
  )
 ]
}
