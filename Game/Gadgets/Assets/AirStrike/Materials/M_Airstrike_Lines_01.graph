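# Translucent unlit particle material for the Airstrike gadget's line effects, usable on mesh and sprite particles.
# Emissive blends two colors by a FriendOrFoe scalar, splits inner and outer intensity at the U midpoint, and multiplies by particle color.
# An opacity chain multiplies Fresnel by optional screen-space panning scanlines and particle alpha.
# A switchable dotted-line mask shifts its V coordinate in discrete steps from a slider.
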Material /Game/Gadgets/Assets/AirStrike/Materials/M_Airstrike_Lines_01
graph {
 properties {
  bEnableResponsiveAA=true
  bUsedWithMeshParticles=true
  bUsedWithParticleSprites=true
  blendMode=BLEND_Translucent
  pixelDepthOffsetMode=PDOM_Legacy
  shadingModel=MSM_Unlit
  translucencyLightingMode=TLM_VolumetricDirectional
  translucentShadowDensityScale=6
  usedShadingModels=Unlit
 }
 nodes [
  Desaturation(
   None=MaterialFunctionCall!(Team, Debug Colors, Debug speed) {materialFunction=/Game/Packages/Fortress_SharedMaterials/Base_Material_Functions/Fortnite_TeamColor}.Result
   Fraction=ScalarParameter() {defaultValue=0.7102, parameterName=Desaturation}
  )
  n1 = VectorParameter() {parameterName=Color, defaultValue=<rgba 7.2102 32.2659 60 0>}
  n2 = VectorParameter() {parameterName=Color B, defaultValue=<rgba 20 11.6996 2.7071 0>}
  n3 = ScalarParameter() {group=Team Colors, parameterName=FriendOrFoe}
  n4 = TextureCoordinate()
  n5 = Ceil(None=Subtract(A=ComponentMask(None=@n4) {r=true}) {constB=0.5})
  n6 = Multiply(
   A=LinearInterpolate(A=LinearInterpolate(A=@n1.RGB, B=@n2.RGB, Alpha=@n3), Alpha=@n5)
   B=ScalarParameter() {defaultValue=0.2, parameterName=Emissive Inner}
  )
  n7 = ScalarParameter() {defaultValue=0.2, parameterName=Emissive Outer}
  n8 = ParticleColor()
  MaterialFunctionCall(Emissive=Multiply(A=LinearInterpolate(A=@n6, B=@n7, Alpha=@n5), B=@n8.RGB)) {materialFunction=&MF_FX_EmissiveAdjustments}
  Multiply(A=MaterialFunctionCall!(FriendOrFoe, Debug Colors, Debug speed) {materialFunction=/Game/Packages/Fortress_SharedMaterials/Base_Material_Functions/Fortnite_FriendOrFoe_TechColor}.Result) {constB=60}
  n9 = ScalarParameter() {defaultValue=1, parameterName=Opacity}
  n10 = Panner(Coordinate=ScreenPosition().ViewportUV) {bFractionalPart=true, speedY=3}
  n11 = ComponentMask(None=VectorParameter() {parameterName=Scanline Repeats, defaultValue=<rgba 1 150 0 1>}.RGB) {g=true, r=true}
  n12 = Frac(None=ComponentMask(None=Multiply(A=@n10, B=@n11)) {g=true})
  n13 = Multiply(A=@n12, B=OneMinus(None=@n12))
  n14 = ScalarParameter() {defaultValue=10, parameterName=Scanline Brightness}
  n15 = Multiply(
   A=MaterialFunctionCall(
    Brightness=ScalarParameter() {defaultValue=1.2, parameterName=Fresnel Brightness}
    Exponent=ScalarParameter() {defaultValue=0.9, parameterName=Fresnel Exponent}
   ) {materialFunction=&MF_Fresnel_FX}.Fresnel
   B=StaticSwitchParameter(True=Multiply(A=@n13, B=@n14), False=Constant() {r=1}) {parameterName=Use Scanlines?}
  )
  Multiply(A=Clamp(None=Multiply(A=@n9, B=@n15) {constA=4}), B=@n8.A)
  n16 = TextureCoordinate() {vTiling=3}
  n17 = ScalarParameter() {defaultValue=10, parameterName=Steps}
  n18 = Ceil(None=Multiply(A=ScalarParameter() {defaultValue=0.5, parameterName=Slider}, B=@n17))
  n19 = AppendVector(A=Constant(), B=Multiply(A=Divide(A=@n18, B=@n17), B=ScalarParameter() {defaultValue=2.5, parameterName=Range}))
  StaticSwitchParameter(
   True=TextureSample(UVs=Subtract(A=@n16, B=@n19)) {samplerType=SAMPLERTYPE_LinearColor, texture=/Game/Gadgets/Assets/AirStrike/Textures/T_Airstrike_LinesMask_01}.RGB
   False=Constant() {r=1}
  ) {parameterName=UseDottedLineFade?}
 ]
}
