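# Translucent unlit mesh-particle material for a simple mushroom cloud effect in the Airstrike gadget, styled with scrolling scanlines.
# An inverted Fresnel blends inner and outer colors and opacities, with the color tinted by particle color and passed through an emissive adjustment function.
# Panned screen-space UVs form repeating soft horizontal bands that modulate opacity, which is then depth-faded and scaled by particle alpha.
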Material /Game/Gadgets/Assets/AirStrike/Materials/M_Airstrike_MushroomCloudSimple_01
graph {
 properties {
  bEnableResponsiveAA=true
  bScreenSpaceReflections=true
  bTangentSpaceNormal=false
  bUsedWithMeshParticles=true
  blendMode=BLEND_Translucent
  numCustomizedUVs=4
  pixelDepthOffsetMode=PDOM_Legacy
  shadingModel=MSM_Unlit
  translucencyLightingMode=TLM_VolumetricDirectional
  usedShadingModels=Unlit
 }
 nodes [
  n1 = VectorParameter() {parameterName=Color Inner, defaultValue=<rgba 0.945 0.5176 0.1418 0>}
  n2 = VectorParameter() {parameterName=Color Outer, defaultValue=<rgba 0.945 0.7975 0.7238 0>}
  n3 = MaterialFunctionCall(
   Brightness=ScalarParameter() {defaultValue=1.25, parameterName=Fresnel Brightness}
   Exponent=ScalarParameter() {defaultValue=1.5, parameterName=Fresnel Exponent}
  ) {materialFunction=&MF_Fresnel_FX}
  MaterialFunctionCall(Emissive=Multiply(A=ParticleColor().RGB, B=LinearInterpolate(A=@n1.RGB, B=@n2.RGB, Alpha=@n3.Fresnel Inverted))) {materialFunction=&MF_FX_EmissiveAdjustments}
  n4 = LinearInterpolate(
   A=ScalarParameter() {defaultValue=0.3, parameterName=Opacity Inner}
   B=ScalarParameter() {defaultValue=1, parameterName=Opacity Outer}
   Alpha=@n3.Fresnel Inverted
  )
  n5 = Panner(Coordinate=ScreenPosition().ViewportUV) {bFractionalPart=true, speedY=3}
  n6 = Constant2Vector() {g=150, r=1}
  n7 = Frac(None=ComponentMask(None=Multiply(A=@n5, B=@n6)) {g=true})
  n8 = OneMinus(None=@n7)
  n9 = Clamp(None=Multiply(A=Multiply(A=@n7, B=@n8), B=ScalarParameter() {defaultValue=4, parameterName=Scanline Boost}))
  Multiply(
   A=DepthFade(Opacity=Multiply(A=@n4, B=@n9), FadeDistance=ScalarParameter() {defaultValue=100, parameterName=Depth Fade Distance})
   B=ParticleColor().A
  )
 ]
}
