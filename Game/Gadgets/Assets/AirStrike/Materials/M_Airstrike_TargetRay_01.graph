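# Translucent, unlit, two-sided material for the target ray of the AirStrike gadget, tinted by a mostly desaturated team color.
# Opacity starts as a scalar parameter multiplied by an adjustable Fresnel term and vertex alpha.
# That value passes through a depth fade to soften intersections, then is scaled by a synced distance fade and particle alpha, so particle systems can control its visibility.
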Material /Game/Gadgets/Assets/AirStrike/Materials/M_Airstrike_TargetRay_01
graph {
 properties {blendMode=BLEND_Translucent, pixelDepthOffsetMode=PDOM_Legacy, shadingModel=MSM_Unlit, twoSided=true, usedShadingModels=Unlit}
 nodes [
  Desaturation(
   None=MaterialFunctionCall!(Team, Debug Colors, Debug speed) {materialFunction=/Game/Packages/Fortress_SharedMaterials/Base_Material_Functions/Fortnite_TeamColor}.Result
   Fraction=Constant() {r=0.75}
  )
  n1 = ScalarParameter() {defaultValue=2, parameterName=Fresnel Brightness}
  n2 = ScalarParameter() {defaultValue=0.5, parameterName=Fresnel Exponent}
  n3 = Multiply(
   A=Multiply(
    A=ScalarParameter() {defaultValue=0.1, parameterName=Opacity}
    B=MaterialFunctionCall(Brightness=@n1, Exponent=@n2) {materialFunction=&MF_Fresnel_FX}.Fresnel
   )
   B=VertexColor().A
  )
  n4 = ScalarParameter() {defaultValue=120, parameterName=Depth Fade Distance}
  n5 = ScalarParameter() {defaultValue=2500, parameterName=start}
  n6 = ScalarParameter() {defaultValue=1300, parameterName=end}
  Multiply(
   A=Multiply(
    A=DepthFade(Opacity=@n3, FadeDistance=@n4)
    B=MaterialFunctionCall(Fade_Start_Distance=@n5, Fade_End_Distance=@n6) {materialFunction=&MF_Distance_SyncedFade_FX}.Result
   )
   B=ParticleColor().A
  )
 ]
}
