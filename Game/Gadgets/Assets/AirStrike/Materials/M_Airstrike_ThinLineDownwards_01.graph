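# Translucent unlit mesh-particle material for the Airstrike gadget's thin downward line effect, previewed on the guidance line mesh.
# A Fresnel blend of two colors brightens toward boosted Color B inside a band moved along V by the dynamic Sweep parameter.
# Particle color tints the emissive.
# Opacity multiplies scanlines, a V-axis mask fading the ends, the sweep band and particle alpha.
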Material /Game/Gadgets/Assets/AirStrike/Materials/M_Airstrike_ThinLineDownwards_01
graph {
 properties {
  bEnableResponsiveAA=true
  bUsedWithMeshParticles=true
  blendMode=BLEND_Translucent
  pixelDepthOffsetMode=PDOM_Legacy
  previewMesh=/Game/Gadgets/Assets/AirStrike/Meshes/Sm_Airstrike_GuidanceLine_01
  shadingModel=MSM_Unlit
  usedShadingModels=Unlit
 }
 nodes [
  n1 = VectorParameter() {parameterName=Color B, defaultValue=<rgba 0.3 4.283 10 1>}
  n2 = LinearInterpolate(
   A=VectorParameter() {parameterName=Color A, defaultValue=<rgba 0.043 0.0193 3 1>}.RGB
   B=@n1.RGB
   Alpha=MaterialFunctionCall(Exponent=ScalarParameter() {defaultValue=5, parameterName=Fresnel Exponent}) {materialFunction=&MF_Fresnel_FX}.Fresnel
  )
  n3 = Multiply(A=@n1.RGB) {constB=5}
  n4 = TextureCoordinate()
  n5 = SphereMask(
   A=Constant() {r=0.5}
   B=Add(A=ComponentMask(None=@n4) {g=true}, B=DynamicParameter() {defaultValue=<rgba 0 1 1 1>, paramNames [Sweep,Param2,Param3,Param4]}.Sweep)
   Radius=ScalarParameter() {defaultValue=0.05, parameterName=SweepSphereMaskRadius}
   Hardness=ScalarParameter() {defaultValue=0.2, parameterName=SweepSphereMaskHardness}
  ) {hardnessPercent=0}
  n6 = ParticleColor()
  MaterialFunctionCall(Emissive=Multiply(A=LinearInterpolate(A=@n2, B=@n3, Alpha=@n5), B=@n6.RGB)) {materialFunction=&MF_FX_EmissiveAdjustments}
  n7 = ScalarParameter() {defaultValue=150, parameterName=Scanline Amount}
  n8 = ScalarParameter() {defaultValue=-0.15, parameterName=ScanlineSize}
  n9 = ScalarParameter() {defaultValue=0.5, parameterName=Scanlines Dark}
  n10 = ScalarParameter() {defaultValue=1, parameterName=Scanlines Light}
  n11 = Constant() {r=0.5}
  n12 = SphereMask(
   A=@n11
   B=ComponentMask(None=TextureCoordinate()) {g=true}
   Radius=@n11
   Hardness=ScalarParameter() {defaultValue=0.8, parameterName=SphereMaskHardness}
  ) {hardnessPercent=0}
  n13 = LinearInterpolate(A=ScalarParameter() {defaultValue=0.1, parameterName=Sweep Low}, Alpha=@n5)
  Multiply(
   A=@n6.A
   B=Multiply(
    A=MaterialFunctionCall!(
     Number of Scanlines=@n7
     Scanline Speed
     "Scanlines Thickness (Bias)"=@n8
     Scanlines Dark Value=@n9
     Scanlines Light Value=@n10
     Block Opacity
     UVs
     Time
     Block Repeat
    ) {materialFunction=/Game/Effects/Fort_Effects/Effects/Atlas/Materials/MF_Atlas_Scanlines_01}.Result
    B=Multiply(A=@n12, B=@n13)
   )
  )
 ]
}
